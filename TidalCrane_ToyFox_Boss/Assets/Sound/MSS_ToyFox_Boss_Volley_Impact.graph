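# One-shot mono MetaSound source, named as the impact sound of the ToyFox boss's volley attack.
# On play, a random get picks one of several energy projectile explosion waves and triggers a random float, which sets the wave player's pitch shift.
# The player's finish ends the source.
# It uses the ToyFox Brute in-air projectile loop's attenuation asset, the in-game SFX sound class and a concurrency count cap.
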MetaSoundSource /TidalCrane_ToyFox_Boss/Assets/Sound/MSS_ToyFox_Boss_Volley_Impact
inputs [
 UE.Source.OnPlay {type=Trigger}
 VolleyImpact {type=WaveAsset:Array}
]
interfaces [UE.OutputFormat.Mono:1.0, UE.Source.OneShot:1.0, UE.Source:1.1]
outputs [
 "UE.OutputFormat.Mono.Audio:0" {type=Audio}
 UE.Source.OneShot.OnFinished {type=Trigger}
]
pages [Default {
  input_defaults {
   UE.Source.OnPlay=null
   VolleyImpact [
    /Game/Sounds/Fort_Weapons/Impacts/EnergyProjectile/Alteration_EnergyProjectile_Explosion_01
    /Game/Sounds/Fort_Weapons/Impacts/EnergyProjectile/Alteration_EnergyProjectile_Explosion_02
    /Game/Sounds/Fort_Weapons/Impacts/EnergyProjectile/Alteration_EnergyProjectile_Explosion_03
    /Game/Sounds/Fort_Weapons/Impacts/EnergyProjectile/Alteration_EnergyProjectile_Explosion_04
    /Game/Sounds/Fort_Weapons/Impacts/EnergyProjectile/Alteration_EnergyProjectile_Explosion_05
   ]
  }
  nodes [
   ArrayRandomGet_0 = Array.Random Get:WaveAsset:Array(Next=^UE.Source.OnPlay, In Array=^VolleyImpact) {also {defaults {Weights []}}}
   n1 = UE.RandomFloat(Next=@ArrayRandomGet_0.On Next, Reset=@ArrayRandomGet_0.On Reset)
   WavePlayer_0 = UE.Wave Player:Mono(Wave Asset=@ArrayRandomGet_0.Value, Play=@n1.On Next, Pitch Shift=@n1.Value) {also {version [1,1]}}
  ]
  links [{dst="@UE.Source.OneShot.OnFinished", src=WavePlayer_0.On Finished}, {dst="@UE.OutputFormat.Mono.Audio:0", src=WavePlayer_0.Out Mono}]
 }]
properties {
 assetImportData=&self:AssetImportData
 attenuationSettings=/TidalCrane_ToyFox_Boss/Assets/Sound/Settings/ATT_ToyFox_Brute_Proj_InAir_Loop
 maxDistance=15500
 sampleRate=48000
 soundClassObject=&Fort_SFX_InGame
 concurrencyOverrides {maxCount=5}
}
