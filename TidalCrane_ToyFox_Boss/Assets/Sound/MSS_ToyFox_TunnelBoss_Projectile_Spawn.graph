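# One-shot mono MetaSound source for the projectile spawn sound of the ToyFox tunnel boss.
# On play, a random pick from a wave asset array, which lists no waves here, triggers a random float that sets the wave player's pitch shift and starts it.
# The player feeds the mono output and finish trigger, with boss-specific attenuation and a concurrency cap applied to the source.
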MetaSoundSource /TidalCrane_ToyFox_Boss/Assets/Sound/MSS_ToyFox_TunnelBoss_Projectile_Spawn
inputs [UE.Source.OnPlay {type=Trigger}]
interfaces [UE.OutputFormat.Mono:1.0, UE.Source.OneShot:1.0, UE.Source:1.1]
outputs [
 "UE.OutputFormat.Mono.Audio:0" {type=Audio}
 UE.Source.OneShot.OnFinished {type=Trigger}
]
pages [Default {
  input_defaults {UE.Source.OnPlay=null}
  nodes [
   ArrayRandomGet_0 = Array.Random Get:WaveAsset:Array(Next=^UE.Source.OnPlay, In Array=VariableAccessor.WaveAsset:Array(Variable=InitVariable.WaveAsset:Array().Variable).Value) {also {defaults {Weights []}}}
   n1 = UE.RandomFloat(Next=@ArrayRandomGet_0.On Next, Reset=@ArrayRandomGet_0.On Reset)
   WavePlayer_0 = UE.Wave Player:Mono(Wave Asset=@ArrayRandomGet_0.Value, Play=@n1.On Next, Pitch Shift=@n1.Value) {also {version [1,1]}}
  ]
  links [{dst="@UE.Source.OneShot.OnFinished", src=WavePlayer_0.On Finished}, {dst="@UE.OutputFormat.Mono.Audio:0", src=WavePlayer_0.Out Mono}]
 }]
properties {
 assetImportData=&self:AssetImportData
 attenuationSettings=/TidalCrane_ToyFox_Boss/Assets/Sound/Settings/ATT_ToyFox_Boss_Tunnel_ProjectileSpawn
 maxDistance=30000
 sampleRate=48000
 soundClassObject=&Fort_SFX_InGame
 volume=4
 concurrencyOverrides {maxCount=5}
}
