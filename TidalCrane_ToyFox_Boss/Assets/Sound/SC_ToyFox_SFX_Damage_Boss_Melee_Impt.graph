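# Sound cue for the ToyFox boss's melee damage impact sound effect in a boss fight.
# A Random node picks one of four equally weighted impact waves, feeding a Modulator at the root for per-play variation.
# It plays through the boss-fight SFX sound class with a roar VOX attenuation asset for distance falloff.
# A concurrency override caps simultaneous instances at a low count, stopping the oldest.
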SoundCue /TidalCrane_ToyFox_Boss/Assets/Sound/SC_ToyFox_SFX_Damage_Boss_Melee_Impt
root=Modulator_0
nodes [
 {id=Modulator_0, children [Random_1]}
 {id=Random_1, children [WavePlayer_3,WavePlayer_4,WavePlayer_5,WavePlayer_7], properties {weights [1,1,1,1]}}
 {id=WavePlayer_3, properties {soundWaveAssetPtr=/TidalCrane_ToyFox_Boss/Assets/Sound/SoundWaves/TidalCrane_Queen_ToyFox_Impact_04}}
 {id=WavePlayer_4, properties {soundWaveAssetPtr=/TidalCrane_ToyFox_Boss/Assets/Sound/SoundWaves/TidalCrane_Queen_ToyFox_Impact_03}}
 {id=WavePlayer_5, properties {soundWaveAssetPtr=/TidalCrane_ToyFox_Boss/Assets/Sound/SoundWaves/TidalCrane_Queen_ToyFox_Impact_02}}
 {id=WavePlayer_7, properties {soundWaveAssetPtr=/TidalCrane_ToyFox_Boss/Assets/Sound/SoundWaves/TidalCrane_Queen_ToyFox_Impact_01}}
]
properties {
 attenuationSettings=/TidalCrane_ToyFox_Boss/Assets/Sound/Settings/ATT_ToyFox_Boss_Roar_VOX
 bOverrideConcurrency=true
 maxDistance=80000
 soundClassObject=&Fort_BossFight_SFX
 volumeMultiplier=0.8
 attenuationOverrides {falloffDistance=25000, attenuationShapeExtents {x=6000}}
 concurrencyOverrides {maxCount=2, resolutionRule=StopOldest}
}
