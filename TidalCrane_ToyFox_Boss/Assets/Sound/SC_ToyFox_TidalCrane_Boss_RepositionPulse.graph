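# Sound cue for the TidalCrane ToyFox boss's reposition pulse, played as boss-fight SFX.
# A Random node picks with equal weight among five BurrowPulse wave variations and plays one at boosted volume.
# It uses the boss roar earth-movement attenuation asset for long-range falloff, with overridden concurrency that caps simultaneous instances and stops the oldest.
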SoundCue /TidalCrane_ToyFox_Boss/Assets/Sound/SC_ToyFox_TidalCrane_Boss_RepositionPulse
root=Random_0
nodes [
 {id=Random_0, children [WavePlayer_0,WavePlayer_1,WavePlayer_2,WavePlayer_3,WavePlayer_4], properties {weights [1,1,1,1,1]}}
 {id=WavePlayer_0, properties {soundWaveAssetPtr=/TidalCrane_ToyFox_Boss/Assets/Sound/SoundWaves/TidalCrane_Queen_ToyFox_BurrowPulse_05}}
 {id=WavePlayer_1, properties {soundWaveAssetPtr=/TidalCrane_ToyFox_Boss/Assets/Sound/SoundWaves/TidalCrane_Queen_ToyFox_BurrowPulse_04}}
 {id=WavePlayer_2, properties {soundWaveAssetPtr=/TidalCrane_ToyFox_Boss/Assets/Sound/SoundWaves/TidalCrane_Queen_ToyFox_BurrowPulse_03}}
 {id=WavePlayer_3, properties {soundWaveAssetPtr=/TidalCrane_ToyFox_Boss/Assets/Sound/SoundWaves/TidalCrane_Queen_ToyFox_BurrowPulse_02}}
 {id=WavePlayer_4, properties {soundWaveAssetPtr=/TidalCrane_ToyFox_Boss/Assets/Sound/SoundWaves/TidalCrane_Queen_ToyFox_BurrowPulse_01}}
]
properties {
 attenuationSettings=/TidalCrane_ToyFox_Boss/Assets/Sound/Settings/ATT_ToyFox_Boss_Roar_EarthMovement
 bOverrideConcurrency=true
 maxDistance=32500
 soundClassObject=&Fort_BossFight_SFX
 volumeMultiplier=2
 attenuationOverrides {falloffDistance=25000, attenuationShapeExtents {x=9000}}
 concurrencyOverrides {maxCount=3, resolutionRule=StopOldest}
}
